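# Niagara system for a small hydrant water burst: an upward jet with splashes, droplets and mist.
# Mesh emitters build the jet, fill and top splash using curve-driven material parameters, while sprite emitters add gloops at base and top plus mist.
# Jet spawn rate and mist launch velocity start high, then fall off for an initial surge.
# User PipeSize scales widths, TopOffset places top gloop and splash, and VelocityMult scales gloop speed.
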NiagaraSystem /BlastBerryMapContent/Game/Environments/Water/Hydrants/NS_Hydrant_WaterBurst_Small
dependencies {
 usedDataInterfaces [
  &NiagaraDataInterfaceCurve
  &NiagaraDataInterfaceColorCurve
  &NiagaraDataInterfaceVectorCurve
  &NiagaraDataInterfaceMeshRendererInfo
 ]
 usedDynamicInputs [&MakeFloatFromVector2d]
 usedModules [
  &self:SystemSpawnScript.SystemScriptSource.SystemScriptGraph.NiagaraNodeAssignment_2.SetVariables_#1
  &SystemState
  &self:SystemSpawnScript.SystemScriptSource.SystemScriptGraph.NiagaraNodeAssignment_3.SetVariables_#2
 ]
}
emitters {
 BottomGloop001 {
  emitterUpdateScript [
   &EmitterState(Delay First Loop Only=true, Life Cycle Mode=::NewEnumerator1, Loop Delay=0.25, Loop Duration=1, UseLoopDelay=true)
   &SpawnRate(SpawnRate=20, Use Spawn Probability=false)
  ]
  particleSpawnScript [
   &InitializeParticle(
    Color=<rgba 0.4204 0.7641 1 1>
    Lifetime=&RandomRangeFloat(Minimum=0.5, Maximum=0.2)
    Position Offset=&Add_Vector(A=<0 0 0>, B=<0 0 0>)
    Sprite Size=&Multiply_Vector2D_ByFloat(Float=0.35, Vector2D=&RandomRangeVector2D(Minimum=<15 25>, Maximum=<15 15>))
    Sprite Size Mode=::NewEnumerator4
    UsePositionOffset=true
    Write Material Random=true
   ) {module=InitializeParticle001}
   &ShapeLocation(
    Cylinder Height=10
    Cylinder Radius=&Multiply_Float(A=$User.PipeSize, B=1.35)
    MaskSpawnGroup=false
    Override Seed=false
    Shape Primitive=::NewEnumerator1
    Surface Only=true
    Use Endcaps In Surface Only Mode=false
    UseOwnerScale=true
   )
   &AddVelocity(Velocity=&Multiply_VectorByFloat(Vector=&RandomRangeVector(Minimum=<0 0 100>, Maximum=<0 0 150>), Float=$User.VelocityMult)) {module=AddVelocity001}
   &AddVelocity(
    Module.Use Falloff Distance=false
    Override Seed=false
    Velocity Mode=::NewEnumerator1
    Velocity Speed=&Multiply_Float(A=&RandomRangeFloat(Minimum=50, Maximum=100), B=$User.VelocityMult)
   ) {module=AddVelocity002}
   &SubUVAnimation(End Frame=3, SubUV Animation Mode:ENiagara_SubUVLookupMode=::NewEnumerator2)
  ]
  particleUpdateScript [
   &ParticleState()
   &GravityForce(Gravity=&Multiply_VectorByFloat(Vector=<0 0 -980>, Float=&RandomRangeFloat(Minimum=1, Maximum=2)))
   &SolveForcesAndVelocity(Clamp Velocity=false, Limit Acceleration=false)
   &ScaleSpriteSize(Scale Factor=&Vector2DFromFloat(Value=&FloatFromCurve(
      FloatCurve=DataInterface() {propertyValues {Curve {keys [
          {arriveTangent=10.0275, interpMode=RCIM_Cubic, leaveTangent=10.0275, tangentMode=RCTM_User, time=0, value=0.4376}
          {arriveTangent=1.6841, interpMode=RCIM_Cubic, leaveTangent=1.6841, tangentMode=RCTM_User, time=1, value=5}
         ]}}}
      Scale Curve=2
     )))
   &ScaleColor(
    Scale Alpha=&FloatFromCurve(FloatCurve=DataInterface() {propertyValues {Curve {keys [<at 0 0>,<at 0.009 0.9316>,<at 0.7102 0.15>,<at 1 0>]}}})
    Scale RGB=&VectorFromCurve(VectorCurve=DataInterface() {propertyValues {XCurve {keys [<at 0 2>,<at 1 0>]}, YCurve {keys [<at 0 2>,<at 1 0>]}, ZCurve {keys [<at 0 2>,<at 1 0>]}}})
    ScaleA=true
    ScaleRGB=true
   )
   &DynamicMaterialParameters(
    Index 0 Param 1=&FloatFromCurve(FloatCurve=DataInterface() {propertyValues {Curve {keys [
         {arriveTangent=-2.026, interpMode=RCIM_Cubic, leaveTangent=-2.026, tangentMode=RCTM_User, time=0.25, value=1}
         {arriveTangent=-0.4434, interpMode=RCIM_Cubic, leaveTangent=-0.4434, tangentMode=RCTM_User, time=1.0011, value=0}
        ]}}})
    Index 0 Param 2=&FloatFromCurve(FloatCurve=DataInterface() {propertyValues {Curve {keys [<at -0.0086 -1.3001>,<at 0.9945 1.9153>]}}})
    Index 0 Param 3=0.5
    Module.Param0WriteEnabled=true
    Module.Param1WriteEnabled=false
    Module.Param2WriteEnabled=true
    Module.Param3WriteEnabled=true
   ) {module=DynamicMaterialParameters001}
   &SpriteRotationRate(Rotation Rate=&Lerp_Float(A=-150, B=150, Alpha=$Particles.Initial.MaterialRandom))
  ]
  rendererClasses [&NiagaraSpriteRendererProperties]
  renderers [{rendererClass=&NiagaraSpriteRendererProperties}]
 }
 TopGloop {
  emitterUpdateScript [
   &EmitterState(Delay First Loop Only=true, Life Cycle Mode=::NewEnumerator1, Loop Delay=0.25, Loop Duration=1, UseLoopDelay=true)
   &SpawnRate(SpawnRate=20, Use Spawn Probability=false)
  ]
  particleSpawnScript [
   &InitializeParticle(
    Color=<rgba 0.4204 0.7641 1 1>
    Lifetime=&RandomRangeFloat(Minimum=0.5, Maximum=0.8)
    Position Offset=&Add_Vector(A=$User.TopOffset, B=<0 0 30>)
    Sprite Size=&Multiply_Vector2D_ByFloat(Float=0.8, Vector2D=&RandomRangeVector2D(Minimum=<15 25>, Maximum=<15 15>))
    Sprite Size Mode=::NewEnumerator4
    UsePositionOffset=true
    Write Material Random=true
   ) {module=InitializeParticle001}
   &ShapeLocation(
    Cylinder Height=50
    Cylinder Radius=&Multiply_Float(A=$User.PipeSize, B=1.35)
    MaskSpawnGroup=false
    Override Seed=false
    Shape Primitive=::NewEnumerator1
    Surface Only=false
    UseOwnerScale=true
   )
   &AddVelocity(Velocity=&Multiply_VectorByFloat(Vector=&RandomRangeVector(Minimum=<0 0 100>, Maximum=<0 0 150>), Float=$User.VelocityMult)) {module=AddVelocity001}
   &AddVelocity(
    Cone Axis=<0 0 1>
    Module.UseConeFalloff=false
    Module.UseVelDistribution=false
    Override Seed=false
    Velocity Mode=::NewEnumerator2
    Velocity Speed=&Multiply_Float(A=&RandomRangeFloat(Minimum=150, Maximum=300), B=$User.VelocityMult)
   ) {module=AddVelocity002}
   &SubUVAnimation(End Frame=3, SubUV Animation Mode:ENiagara_SubUVLookupMode=::NewEnumerator2)
  ]
  particleUpdateScript [
   &ParticleState()
   &GravityForce(Gravity=&Multiply_VectorByFloat(Vector=<0 0 -980>, Float=&RandomRangeFloat(Minimum=1, Maximum=2)))
   &SolveForcesAndVelocity(Clamp Velocity=false, Limit Acceleration=false)
   &ScaleSpriteSize(Scale Factor=&Vector2DFromFloat(Value=&FloatFromCurve(
      FloatCurve=DataInterface() {propertyValues {Curve {keys [
          {arriveTangent=10.0275, interpMode=RCIM_Cubic, leaveTangent=10.0275, tangentMode=RCTM_User, time=0, value=0.4376}
          {arriveTangent=1.6841, interpMode=RCIM_Cubic, leaveTangent=1.6841, tangentMode=RCTM_User, time=1, value=5}
         ]}}}
      Scale Curve=2
     )))
   &ScaleColor(
    Scale Alpha=&FloatFromCurve(FloatCurve=DataInterface() {propertyValues {Curve {keys [<at 0 0>,<at 0.009 0.9316>,<at 0.7102 0.15>,<at 1 0>]}}})
    Scale RGB=&VectorFromCurve(VectorCurve=DataInterface() {propertyValues {XCurve {keys [<at 0 2>,<at 1 0>]}, YCurve {keys [<at 0 2>,<at 1 0>]}, ZCurve {keys [<at 0 2>,<at 1 0>]}}})
    ScaleA=true
    ScaleRGB=true
   )
   &DynamicMaterialParameters(
    Index 0 Param 1=&FloatFromCurve(FloatCurve=DataInterface() {propertyValues {Curve {keys [
         {arriveTangent=-2.026, interpMode=RCIM_Cubic, leaveTangent=-2.026, tangentMode=RCTM_User, time=0.25, value=1}
         {arriveTangent=-0.4434, interpMode=RCIM_Cubic, leaveTangent=-0.4434, tangentMode=RCTM_User, time=1.0011, value=0}
        ]}}})
    Index 0 Param 2=&FloatFromCurve(FloatCurve=DataInterface() {propertyValues {Curve {keys [<at -0.0086 -1.3001>,<at 0.9945 1.9153>]}}})
    Index 0 Param 3=0.5
    Module.Param0WriteEnabled=true
    Module.Param1WriteEnabled=false
    Module.Param2WriteEnabled=true
    Module.Param3WriteEnabled=true
   ) {module=DynamicMaterialParameters001}
   &SpriteRotationRate(Rotation Rate=&Lerp_Float(A=-150, B=150, Alpha=$Particles.Initial.MaterialRandom))
  ]
  rendererClasses [&NiagaraSpriteRendererProperties]
  renderers [{rendererClass=&NiagaraSpriteRendererProperties}]
 }
 TopSplash {
  emitterUpdateScript [
   &EmitterState(Delay First Loop Only=true, Life Cycle Mode=::NewEnumerator1, Loop Delay=0.5, Loop Duration=1, UseLoopDelay=true)
   &SpawnRate(SpawnRate=10, Use Spawn Probability=false) {module=SpawnRate001}
  ]
  particleSpawnScript [
   &InitializeParticle(
    Lifetime Max=0.8
    Lifetime Min=0.6
    Lifetime Mode=::NewEnumerator1
    Mesh Scale Max=<0.15 0.15 0.17>
    Mesh Scale Min=<0.07 0.07 0.1>
    Mesh Scale Mode=::NewEnumerator2
    Position Offset=&Add_Vector(A=$User.TopOffset, B=<0 0 15>)
    UsePositionOffset=true
    Write Material Random=true
   ) {module=InitializeParticle001}
   &ShapeLocation(
    Cylinder Height=0
    Cylinder Radius=&Multiply_Float(A=$User.PipeSize, B=1.35)
    MaskSpawnGroup=false
    Override Seed=false
    Shape Primitive=::NewEnumerator1
    Surface Only=false
    UseOwnerScale=true
   )
   &InitialMeshOrientation(
    Mesh Orientation Mode=::NewEnumerator2
    Rotation=&MakeVector(X=0, Y=0, Z=&RandomRangeFloat(Minimum=0, Maximum=1))
    Use Rotation Vector=true
   )
  ]
  particleUpdateScript [
   &ParticleState()
   &DynamicMaterialParameters(
    Index 0 Param 1=&FloatFromCurve(FloatCurve=DataInterface() {propertyValues {Curve {keys [
         {arriveTangent=0.8954, interpMode=RCIM_Cubic, leaveTangent=0.8954, tangentMode=RCTM_User, time=-0.0005, value=-1.845}
         {arriveTangent=1.8419, interpMode=RCIM_Cubic, leaveTangent=1.8419, time=0.7887, value=-0.7711}
         {arriveTangent=3.1029, interpMode=RCIM_Cubic, leaveTangent=3.1029, tangentMode=RCTM_User, time=1.0011, value=0}
        ]}}})
    Index 0 Param 2=&FloatFromCurve(FloatCurve=DataInterface() {propertyValues {Curve {keys [<at -0.0046 4>,<at 1.0011 0>]}}})
    Index 0 Param 3=&FloatFromCurve(
     FloatCurve=DataInterface() {propertyValues {Curve {keys [
         {arriveTangent=0.096, interpMode=RCIM_Cubic, leaveTangent=0.096, tangentMode=RCTM_User, time=-0.0046, value=0.0882}
         <at 1.0011 0.1372 RCIM_Cubic>
        ]}}}
     Scale Curve=&RandomRangeFloat(Minimum=1.5, Maximum=2.5)
    )
    Index 0 Param 4=&FloatFromCurve(
     FloatCurve=DataInterface() {propertyValues {Curve {keys [
         {arriveTangent=1.9125, interpMode=RCIM_Cubic, leaveTangent=0.2728, tangentMode=RCTM_User, time=-0.0046, value=0.9276}
         {arriveTangent=4.6326, interpMode=RCIM_Cubic, leaveTangent=4.6326, tangentMode=RCTM_User, time=1.0011, value=2.7648}
        ]}}}
     Scale Curve=&RandomRangeFloat(Minimum=1.5, Maximum=3)
    )
    Module.Param0WriteEnabled=true
    Module.Param0WriteEnabled3=true
    Module.Param1WriteEnabled=true
    Module.Param1WriteEnabled3=true
    Module.Param2WriteEnabled=true
    Module.Param2WriteEnabled3=true
    Module.Param3WriteEnabled=true
    Module.Param3WriteEnabled3=true
    Write Parameter Index 3=true
   ) {module=DynamicMaterialParameters001}
   &ScaleColor(
    Scale Alpha=&FloatFromCurve(FloatCurve=DataInterface() {propertyValues {Curve {keys [<at -0.0046 1.0113 RCIM_Cubic>,<at 1.0011 0.002 RCIM_Cubic>]}}})
    ScaleA=true
    ScaleRGB=false
   )
   &ScaleMeshSize(Scale Factor=&VectorFromCurve(VectorCurve=DataInterface() {propertyValues {XCurve {keys [<at 0 1>,<at 1 3>]}, YCurve {keys [<at 0 1>,<at 1 3>]}, ZCurve {keys [<at 0 1>,<at 0.9989 1.5622>]}}})) {module=MeshSizeScale}
   &ScaleMeshSize(Scale Factor=&MakeVector(X=0.5, Y=0.5, Z=2))
  ]
  rendererClasses [&NiagaraMeshRendererProperties]
  renderers [{rendererClass=&NiagaraMeshRendererProperties, rendererIndex=0}]
 }
 WaterJetFill {
  emitterUpdateScript [
   &EmitterState(
    Delay First Loop Only=true
    Life Cycle Mode=::NewEnumerator1
    Loop Delay=0.1
    Loop Duration=1
    Scalability Mode=::NewEnumerator1
    UseLoopDelay=true
   )
   &SpawnRate(SpawnRate=6, Use Spawn Probability=false)
  ]
  particleSpawnScript [
   &InitializeParticle(
    Color=<rgba 1 1 1 1.25>
    Lifetime=&RandomRangeFloat(Minimum=0.6, Maximum=0.8)
    Mesh Scale=<0.1 0.1 0.2>
    Mesh Scale Mode=::NewEnumerator4
    Position Offset=<0 0 -200>
    UsePositionOffset=true
    Write Material Random=true
   )
   &AddVelocity(Velocity=&Multiply_VectorByFloat(Vector=&RandomRangeVector(Minimum=<0 0 100>, Maximum=<0 0 50>), Float=1))
  ]
  particleUpdateScript [
   &ParticleState()
   &DynamicMaterialParameters(
    Index 0 Param 1=&FloatFromCurve(
     FloatCurve=DataInterface() {propertyValues {Curve {keys [
         {arriveTangent=2.0382, interpMode=RCIM_Cubic, leaveTangent=2.0382, tangentMode=RCTM_User, time=0, value=0}
         {arriveTangent=0.7509, interpMode=RCIM_Cubic, leaveTangent=0.7509, tangentMode=RCTM_User, time=1, value=1}
        ]}}}
     Scale Curve=8
    )
    Index 0 Param 2=&FloatFromCurve(
     FloatCurve=DataInterface() {propertyValues {Curve {keys [{interpMode=RCIM_Cubic, tangentMode=RCTM_User, time=0, value=0}, {interpMode=RCIM_Cubic, tangentMode=RCTM_User, time=1, value=1}]}}}
     Scale Curve=2.25
    )
    Index 0 Param 3=&FloatFromCurve(
     FloatCurve=DataInterface() {propertyValues {Curve {keys [{interpMode=RCIM_Cubic, tangentMode=RCTM_User, time=0, value=1}, {interpMode=RCIM_Cubic, tangentMode=RCTM_User, time=1, value=0}]}}}
     Scale Curve=8
    )
    Index 0 Param 4=3
    Module.Param0WriteEnabled=true
    Module.Param1WriteEnabled=true
    Module.Param2WriteEnabled=true
    Module.Param3WriteEnabled=true
   )
   &SolveForcesAndVelocity(Clamp Velocity=false, Limit Acceleration=false)
   &ScaleMeshSize(Scale Factor=&MakeVector(
     X=&Multiply_Float(A=0.08, B=$User.PipeSize)
     Y=&Multiply_Float(A=0.08, B=$User.PipeSize)
     Z=1
    ))
  ]
  rendererClasses [&NiagaraMeshRendererProperties]
  renderers [{rendererClass=&NiagaraMeshRendererProperties, rendererIndex=0}]
 }
 WaterJets {
  emitterUpdateScript [
   &EmitterState(Delay First Loop Only=true, Life Cycle Mode=::NewEnumerator1, Loop Duration=1, UseLoopDelay=true)
   &SpawnRate(
    SpawnRate=&Lerp_Float(
     A=12
     B=50
     Alpha=&FloatFromCurve(
      FloatCurve=DataInterface() {propertyValues {Curve {keys [{tangentMode=RCTM_User, time=0, value=1}, {tangentMode=RCTM_User, time=0.5, value=0}]}}}
      CurveIndex=$Emitter.Age
     )
    )
    Use Spawn Probability=false
   ) {module=SpawnRate001}
  ]
  particleSpawnScript [
   &InitializeParticle(
    Color=<rgba 0.3021 0.3828 1 1>
    Lifetime=&RandomRangeFloat(Minimum=1, Maximum=1.2)
    Mesh Scale Mode=::NewEnumerator3
    Position Offset=<0 0 -100>
    UsePositionOffset=true
    Write Material Random=true
   ) {module=InitializeParticle001}
   /SetVariables(Particles.MeshOrientation=&MakeQuatFromAxisAngle(Axis=<1 0 0>, AngleInDegrees=&RandomRangeFloat(Minimum=0, Maximum=350))) {isSetParameters=true}
   &AddVelocityInCone(
    Cone Angle=0
    Cone Axis=<0 0 1>
    Cone Axis Coordinate Space=::World
    Module.Use Velocity Falloff On Cone Axis=false
    Velocity Strength=&Multiply_Float(
     A=$System.Startin
     B=&Multiply_Float(
      A=3200
      B=&Multiply_Float(A=&RandomRangeFloat(Minimum=0.6, Maximum=0.6), B=1)
     )
    )
   )
  ]
  particleUpdateScript [
   &DynamicMaterialParameters(
    Index 0 Param 1=&FloatFromCurve(
     FloatCurve=DataInterface() {propertyValues {Curve {keys [
         {arriveTangent=1, interpMode=RCIM_Cubic, leaveTangent=1, time=0, value=0}
         {arriveTangent=1, interpMode=RCIM_Cubic, leaveTangent=1, time=1, value=1}
        ]}}}
     Scale Curve=-1
    )
    Index 0 Param 2=&FloatFromCurve(FloatCurve=DataInterface() {propertyValues {Curve {keys [
         {arriveTangent=-0.3125, interpMode=RCIM_Cubic, leaveTangent=-0.3125, time=0, value=1}
         {arriveTangent=-0.75, interpMode=RCIM_Cubic, leaveTangent=-0.75, time=0.4, value=0.875}
         {arriveTangent=-1.0417, interpMode=RCIM_Cubic, leaveTangent=-1.0417, time=1, value=0.25}
        ]}}})
    Index 0 Param 3=&FloatFromCurve(FloatCurve=DataInterface() {propertyValues {Curve {keys [
         {arriveTangent=-0.375, interpMode=RCIM_Cubic, leaveTangent=-0.375, time=0, value=0.5}
         {arriveTangent=-0.375, interpMode=RCIM_Cubic, leaveTangent=-0.375, time=1, value=2}
        ]}}})
    Index 0 Param 4=&FloatFromCurve(FloatCurve=DataInterface() {propertyValues {Curve {keys [
         {arriveTangent=-1.0556, interpMode=RCIM_Cubic, leaveTangent=-1.0556, time=0, value=1}
         {arriveTangent=-1, interpMode=RCIM_Cubic, leaveTangent=-1, time=0.9, value=0.05}
         {arriveTangent=-0.5, interpMode=RCIM_Cubic, leaveTangent=-0.5, time=1, value=0}
        ]}}})
    Module.Param0WriteEnabled=true
    Module.Param1WriteEnabled=true
    Module.Param2WriteEnabled=true
    Module.Param3WriteEnabled=true
   )
   &ParticleState()
   &Drag(Drag=10, Use Linear Drag=true, Use Rotational Drag=false)
   &SolveForcesAndVelocity(Clamp Velocity=false, Limit Acceleration=false)
   /SetVariables(Particles.WaterXY:NiagaraFloat=&FloatFromCurve(
     FloatCurve=DataInterface() {propertyValues {Curve {keys [
         {arriveTangent=3.4375, interpMode=RCIM_Cubic, leaveTangent=3.4375, time=0, value=0.875}
         {arriveTangent=1.5909, interpMode=RCIM_Cubic, leaveTangent=1.5909, time=0.4, value=2.25}
         {arriveTangent=0.5357, interpMode=RCIM_Cubic, leaveTangent=0.5357, time=1.1, value=2.625}
        ]}}}
     Scale Curve=0.5
    )) {isSetParameters=true}
   &ScaleMeshSize(Scale Factor=&MakeVector(
     X=&FloatFromCurve(FloatCurve=DataInterface() {propertyValues {Curve {keys [
          {arriveTangent=-7, interpMode=RCIM_Cubic, leaveTangent=-7, time=0, value=4}
          {arriveTangent=-5.7143, interpMode=RCIM_Cubic, leaveTangent=-5.7143, time=0.5, value=0.5}
          {arriveTangent=-2.5, interpMode=RCIM_Cubic, leaveTangent=-2.5, time=0.7, value=0}
         ]}}})
     Y=$Particles.WaterXY
     Z=$Particles.WaterXY
    ))
   &ScaleMeshSize(Scale Factor=&VectorFromFloat(Value=0.3)) {module=ScaleMeshSize001}
   &ScaleMeshSize(Scale Factor=&MakeVector(
     X=&Multiply_Float(A=0.1, B=$User.PipeSize)
     Y=&Multiply_Float(A=0.1, B=$User.PipeSize)
     Z=1
    )) {module=ScaleMeshSize002}
   &Color(Color=&ColorFromCurve(ColorCurve=DataInterface() {propertyValues {
       AlphaCurve {keys [
         {arriveTangent=-2.6667, interpMode=RCIM_Cubic, leaveTangent=-2.6667, time=0, value=1}
         {arriveTangent=-1, interpMode=RCIM_Cubic, leaveTangent=-1, time=0.3, value=0.2}
         {arriveTangent=-0.2857, interpMode=RCIM_Cubic, leaveTangent=-0.2857, time=1, value=0}
        ]}
       BlueCurve {keys [{arriveTangent=-2.9961, interpMode=RCIM_Cubic, time=0, value=3}, <at 0.22 0.25>]}
       GreenCurve {keys [{arriveTangent=-0.7617, interpMode=RCIM_Cubic, time=0, value=2.6062}, <at 0.22 0.25>]}
       RedCurve {keys [{arriveTangent=0.2336, interpMode=RCIM_Cubic, time=0, value=1.6261}, <at 0.22 0.25>]}
      }}))
   &ScaleColor(Color Value To Scale=$Particles.Color, Scale RGB=<0.35 0.35 0.35>, ScaleA=true, ScaleRGB=true)
  ]
  rendererClasses [&NiagaraMeshRendererProperties]
  renderers [{rendererClass=&NiagaraMeshRendererProperties, rendererIndex=0}]
 }
 WaterTopMist_hard {
  emitterUpdateScript [
   &EmitterState(Delay First Loop Only=true, Life Cycle Mode=::NewEnumerator1, Loop Duration=1, UseLoopDelay=true)
   &SpawnBurst_Instantaneous(Spawn Count=15, Spawn Group=1, Use Spawn Probability=false, UseLoopCountLimit=true)
   &SpawnRate(SpawnRate=10, Use Spawn Probability=false)
  ]
  particleSpawnScript [
   &InitializeParticle(
    Color=<rgba 1 1 1 5>
    Lifetime=&MaskFloatBySpawnGroup(
     Masked Float=&RandomRangeFloat(Minimum=0.75, Maximum=0.5)
     Passthrough Float=&RandomRangeFloat(Minimum=1, Maximum=1)
    )
    Position Offset=&MakeVector(
     X=0
     Y=0
     Z=&FloatFromCurve(FloatCurve=DataInterface() {propertyValues {Curve {keys [<at 0 -100>,<at 0.01 0>]}}}, CurveIndex=$Emitter.Age)
    )
    Sprite Rotation Mode=::NewEnumerator2
    Sprite Size Mode=::NewEnumerator3
    Uniform Sprite Size=&Multiply_Float(A=0.35, B=&RandomRangeFloat(Minimum=200, Maximum=200))
    UsePositionOffset=true
    Write Material Random=true
   )
   &CylinderLocation(
    Cylinder Height Midpoint=0
    Cylinder Radius=35
    MaskSpawnGroup=false
    Offset=<0 0 50>
    Offset Mode:ENiagara_OffsetMode=::NewEnumerator0
    Override Seed=false
    Surface Only=true
    Use Endcaps In Surface Only Mode=false
    UseOwnerScale=true
   )
   &DynamicMaterialParameters(
    Index 0 Param 1=&RandomRangeFloat(Minimum=0, Maximum=0)
    Module.Param0WriteEnabled=true
    Module.Param1WriteEnabled=true
    Module.Param2WriteEnabled=true
    Module.Param3WriteEnabled=true
   )
   &AddVelocity(Velocity=&Multiply_VectorByFloat(
     Vector=&RandomRangeVector(Minimum=<0 0 800>, Maximum=<0 0 1200>)
     Float=&FloatFromCurve(FloatCurve=DataInterface() {propertyValues {Curve {keys [<at 0.0577 1>,<at 0.1577 0>]}}}, CurveIndex=$Engine.System.Age)
    ))
  ]
  particleUpdateScript [
   &ParticleState()
   &GravityForce(Gravity=&MakeVector(
     X=0
     Y=0
     Z=&MaskFloatBySpawnGroup(Masked Float=-980, Passthrough Float=&Multiply_Float(A=-980, B=1.5))
    ))
   &Drag(Drag=&RandomRangeFloat(Minimum=1, Maximum=1), Use Linear Drag=true, Use Rotational Drag=false)
   &SolveForcesAndVelocity(Clamp Velocity=false, Limit Acceleration=false)
   &SubUVAnimation(
    Start Frame=20
    SubUV Animation Mode:ENiagara_SubUVLookupMode=::NewEnumerator1
    SubUV Curve=DataInterface() {propertyValues {Curve {keys [{interpMode=RCIM_Cubic, tangentMode=RCTM_User, time=0, value=0}, {interpMode=RCIM_Cubic, tangentMode=RCTM_User, time=1, value=1}]}}}
   ) {enabled=false}
   &ScaleColor(
    Scale Alpha=&FloatFromCurve(
     FloatCurve=DataInterface() {propertyValues {Curve {keys [
         {arriveTangent=-0.9969, interpMode=RCIM_Cubic, leaveTangent=-0.9969, time=0, value=1}
         {arriveTangent=-0.9969, interpMode=RCIM_Cubic, leaveTangent=-0.9969, time=1.0011, value=0.002}
        ]}}}
     Scale Curve=0.5
    )
    Scale RGB=&MakeVectorFromLinearColorRGB(Color=&ColorFromCurve(ColorCurve=DataInterface() {propertyValues {
        AlphaCurve {keys []}
        BlueCurve {keys [<at 0 0.943 RCIM_Cubic>,<at 0.9345 0.943 RCIM_Cubic>]}
        GreenCurve {keys [
          {arriveTangent=0.2881, interpMode=RCIM_Cubic, leaveTangent=0.2881, time=0, value=0.6737}
          {arriveTangent=0.2881, interpMode=RCIM_Cubic, leaveTangent=0.2881, time=0.9345, value=0.943}
         ]}
        RedCurve {keys [
          {arriveTangent=0.7827, interpMode=RCIM_Cubic, leaveTangent=0.7827, time=0, value=0.2115}
          {arriveTangent=0.7827, interpMode=RCIM_Cubic, leaveTangent=0.7827, time=0.9345, value=0.943}
         ]}
       }}))
    ScaleA=true
    ScaleRGB=true
   )
   &Color(Color=&ColorFromCurve(ColorCurve=DataInterface() {propertyValues {
       AlphaCurve {keys [
         {arriveTangent=0.1875, interpMode=RCIM_Cubic, time=0, value=0}
         <at 0.4 1 RCIM_Cubic>
         {interpMode=RCIM_Cubic, leaveTangent=-0.125, time=1, value=0}
        ]}
       BlueCurve {keys [<at 0.0524 1>,<at 0.8223 0.849>]}
       GreenCurve {keys [<at 0.0524 1>,<at 0.8223 0.7943>]}
       RedCurve {keys [<at 0.0524 1>,<at 0.8223 0.7782>]}
      }}))
   &ScaleSpriteSize(Uniform Curve Sprite Scale=DataInterface() {propertyValues {Curve {keys [
        {arriveTangent=1, interpMode=RCIM_Cubic, leaveTangent=1, time=0, value=0}
        {arriveTangent=1, interpMode=RCIM_Cubic, leaveTangent=1, time=0.3, value=0.3}
        {arriveTangent=1, interpMode=RCIM_Cubic, leaveTangent=1, time=1, value=1}
       ]}}})
   &SpriteRotationRate(Rotation Rate=&RandomRangeFloat(Minimum=-50, Maximum=50))
  ]
  rendererClasses [&NiagaraSpriteRendererProperties]
  renderers [{rendererClass=&NiagaraSpriteRendererProperties}]
 }
}
summary {
 systemName=NS_Hydrant_WaterBurst_Small
 emitters [
  {emitterName=WaterJets, rendererClasses [&NiagaraMeshRendererProperties]}
  {emitterName=WaterJetFill, rendererClasses [&NiagaraMeshRendererProperties]}
  {emitterName=TopGloop, rendererClasses [&NiagaraSpriteRendererProperties]}
  {emitterName=TopSplash, rendererClasses [&NiagaraMeshRendererProperties]}
  {emitterName=WaterTopMist_hard, rendererClasses [&NiagaraSpriteRendererProperties]}
  {emitterName=BottomGloop001, rendererClasses [&NiagaraSpriteRendererProperties]}
 ]
 userVariables [
  User.PipeSize {type=NiagaraFloat, defaultValue=8, description=""}
  User.SpoutHeight {type=NiagaraFloat, defaultValue=0.1, description=""}
  User.TopOffset {type=Vector3f, description="", defaultValue {struct=Vector3f, value=<0 0 120>}}
  User.VelocityMult {type=NiagaraFloat, defaultValue=1, description=""}
 ]
}
